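# Niagara effect for Hilda's broom-and-pumpkin emote prop: a green magical flame burst and smoke reveal with glowing eyes.
# One-shot flame, light, wave, decal and smoke emitters follow bones of User.SkeletalMesh, some spread across bones by spawn group.
# Flame-eye meshes persist on sockets, while a bone-following ribbon trail spawns based on owner speed.
# Embers, sparkles, continuous flames and an alternate trail are disabled.
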NiagaraSystem /BRCosmetics/Animation/Game/MainPlayer/Emotes/Hilda/FX/NS_HildaPumpkin
dependencies {
 usedDataInterfaces [&NiagaraDataInterfaceCurve, &NiagaraDataInterfaceColorCurve]
 usedDynamicInputs []
 usedModules [&SystemState]
}
emitters {
 Embers {
  bEnabled=false
  emitterUpdateScript [
   &EmitterState(Life Cycle Mode=::NewEnumerator1, Loop Behavior=::NewEnumerator1, Loop Duration=0.3, UseLoopDelay=false)
   &SpawnRate(
    Spawn Probability=&UniformRangedFloat(Minimum=0, Maximum=1)
    SpawnRate=&UniformRangedFloat(Minimum=0, Maximum=200)
    Use Spawn Probability=true
   )
  ]
  particleSpawnScript [
   /Niagara/Modules/Spawn/Initialization/InitializeParticle(
    Lifetime=&UniformRangedFloat(Minimum=0.3, Maximum=0.7)
    Module.Write Color=true
    Module.Write Lifetime=true
    Module.Write Mass=true
    Module.Write Material Random=true
    Module.Write Position=true
    Module.Write Scale=false
    Module.Write SpriteRotation=false
    Module.Write SpriteSize=true
    Position=&SimulationPosition()
    Sprite Size=&Vector2DFromFloat(Value=&FloatFromCurve(
      FloatCurve=DataInterface() {propertyValues {Curve {keys [<at 0 1.2>,<at 0.1257 0.9847>,<at 0.2996 0.5>,<at 1 0.3>]}}}
      CurveIndex=&UniformRangedFloat(Minimum=0.5, Maximum=1)
      Scale Curve=3
     ))
   )
   &SphereLocation(
    MaskSpawnGroup=false
    Offset=<0 0 15>
    Override Seed=false
    Sphere Radius=10
    Surface Only=false
    UseNonUniformScale=false
    UseOffset=true
    UseSphereOrientation=false
   )
  ]
  particleUpdateScript [
   &ParticleState()
   &SkeletalMeshLocation(
    Bone Sampling Mode=::NewEnumerator4
    Inherit Velocity=true
    MaskSpawnGroup=false
    Sampled Position Coordinate Space=::Local
    Sampled Position Offset Coordinate Space:ENiagaraMeshTransforms=::NewEnumerator2
    Sampled Velocity Coordinate Space:ENiagaraCoordinateSpace=::Local
    Skeletal Mesh=DataInterface() {propertyValues {
      DefaultMesh=null
      ExcludeBoneName=root
      OptionalComponentClass=null
      PreviewMesh=/BRCosmetics/Animation/Props/Broom_and_Pumpkin/Meshes/Broom_and_Pumpkin_Gadget
      SkinningMode=SkinOnTheFly
      SoftSourceActor=null
      SourceMode=Default
      UvSetIndex=0
      WholeMeshLOD=-1
      bExcludeBone=false
      bReadDeformedGeometry=true
      bRequireCurrentFrameData=true
      ComponentTags []
      FilteredBones [Pumpkin_02,bristle_02]
      FilteredSockets [RightEyeSocket, LeftEyeSocket]
      MeshUserParameter {parameter {name=None, typeDefHandle {registeredTypeIndex=1}, varData []}}
      SamplingRegions []
     }}
    VelocitySpeedLimit=false
   ) {enabled=false}
   &AccelerationForce(Acceleration=&UniformRangedVector(Minimum=<35 0 15>, Maximum=<40 1 25>))
   &Drag(
    Drag=&FloatFromCurve(FloatCurve=DataInterface() {propertyValues {Curve {keys [<at 0 0.5>,<at 0.5 0.5>,<at 1 1>]}}}, Scale Curve=9)
    Use Linear Drag=true
    Use Rotational Drag=false
   )
   /Niagara/Modules/Update/Forces/CurlNoiseForce(
    Noise Frequency=0.04
    Noise Strength=&FloatFromCurve(FloatCurve=DataInterface() {propertyValues {Curve {keys [<at 0 1>,<at 1 0>]}}}, Scale Curve=500)
    Pan Noise=true
    Pan Noise Field=<0 0 1>
   )
   &SolveForcesAndVelocity(Clamp Velocity=false, Limit Acceleration=false)
   &ScaleSpriteSizeBySpeed(
    Max Scale Factor=<2 5>
    Sample Scale Factor By Curve=true
    Scale Factor Curve=DataInterface() {propertyValues {Curve {keys [
        <at -0.0077 0.0765 RCIM_Cubic>
        {arriveTangent=1.2463, interpMode=RCIM_Cubic, leaveTangent=1.2463, tangentMode=RCTM_User, time=0.9985, value=0.9549}
       ]}}}
    Velocity Threshold=500
   )
   &ScaleColor(
    Scale Alpha=&FloatFromCurve(FloatCurve=DataInterface() {propertyValues {Curve {keys [<at 0 1>,<at 1 0>]}}})
    Scale RGB=&MakeVectorFromLinearColorRGB(Color=<rgba 0.7797 5 0.2401 1>)
    ScaleA=true
    ScaleRGB=true
   )
  ]
  rendererClasses [&NiagaraSpriteRendererProperties]
  renderers [{rendererClass=&NiagaraSpriteRendererProperties}]
 }
 Green_Broom_Trail {
  emitterUpdateScript [
   &EmitterState(
    Delay First Loop Only=true
    Inactive Response=::NewEnumerator1
    Life Cycle Mode=::NewEnumerator1
    Loop Delay=0.1
    Loop Duration=1
    UseLoopDelay=true
   )
   &FN_SpawnRibbon_v2(
    Spawn Mode=::NewEnumerator0
    SpawnRate:NiagaraFloat=&MakeCustomFloatFromBool(
     Bool=&SetBoolByFloatComparison(A=&VectorLength(Vector=$Engine.Owner.Velocity), B=100)
     True Float=35
     False Float=0
    )
   )
  ]
  particleSpawnScript [
   &FN_InitializeRibbon_v3(Enable Skeletal Position=true, Facing Axis:ENiagaraOrientationAxis=::ZAxis, Lifetime=1, Ribbon Width=25)
   /SetVariables(Particles.RibbonFacing=&RotateVectorByQuaternion(Vector To Rotate=<0 0 1>, Quaternion=$Particles.SampledRotation)) {isSetParameters=true, module=SetVariables_#1}
   &ApplyInitialForces()
  ]
  particleUpdateScript [
   &ParticleState()
   /Game/Effects/Niagara/Modules/Contrails/FN_UpdateRibbon/FN_UpdateRibbon_v3(Enable Skeletal Position=true)
   /SetVariables(Particles.Position=&Add_Vector(
     A=Linked() {linkedVariable {name=Particles.Position, type=NiagaraPosition}}
     B=&Multiply_VectorByFloat(
      Vector=&RotateVectorByQuaternion(Vector To Rotate=<2 0 0>, Quaternion=$Particles.SampledRotation)
      Float=$Particles.NormalizedAge
     )
    )) {isSetParameters=true, module=SetVariables_#2}
   &Color(
    Color=<rgba 0.9479 1 1 1>
    Scale Alpha=&FloatFromCurve(FloatCurve=DataInterface() {propertyValues {Curve {keys [{interpMode=RCIM_Cubic, leaveTangent=-2.6112, tangentMode=RCTM_Break, time=0, value=1}, <at 1 0>]}}})
   )
   &DynamicMaterialParameters(
    Index 0 Param 3=1
    Module.Param0WriteEnabled=true
    Module.Param1WriteEnabled=true
    Module.Param2WriteEnabled=true
    Module.Param3WriteEnabled=true
   )
  ]
  rendererClasses [&NiagaraRibbonRendererProperties]
  renderers [{rendererClass=&NiagaraRibbonRendererProperties}]
 }
 Green_Broom_Trail001 {
  bEnabled=false
  emitterUpdateScript [
   &EmitterState(
    Delay First Loop Only=true
    Inactive Response=::NewEnumerator1
    Life Cycle Mode=::NewEnumerator1
    Loop Delay=0.1
    Loop Duration=1
    UseLoopDelay=true
   )
   &FN_SpawnRibbon_v2(Spawn Mode=::NewEnumerator0, SpawnRate=40)
  ]
  particleSpawnScript [
   &FN_InitializeRibbon_v3(Enable Skeletal Position=true, Facing Axis:ENiagaraOrientationAxis=::ZAxis, Ribbon Width=25)
   /SetVariables(Particles.RibbonFacing=&RotateVectorByQuaternion(Vector To Rotate=<0 0 1>, Quaternion=$Particles.SampledRotation)) {isSetParameters=true, module=SetVariables_#1}
   &ApplyInitialForces()
  ]
  particleUpdateScript [
   &ParticleState()
   &FN_UpdateRibbon_v3(Enable Skeletal Position=true)
   /Game/Effects/Niagara/Modules/Materials/FN_RibbonMaterial/FN_RibbonMaterial(
    Color Pan U Speed=1
    Color Scale U=0.5
    Erosion Amount:NiagaraFloat=&FloatFromCurve(FloatCurve=DataInterface() {propertyValues {Curve {keys [<at 0 0.5>,<at 1 1.25>]}}})
    Erosion Pan Speed U=0.35
    Jitter by Particles MaterialRandom=true
    Macro Distortion Amount=1
    Macro Pan Speed=1
    Micro Distortion Amount=1
    Micro Pan Speed=1
    UV0 Jitter multiplier=0
   )
   /SetVariables(Particles.Position=&Add_Vector(
     A=Linked() {linkedVariable {name=Particles.Position, type=NiagaraPosition}}
     B=&Multiply_VectorByFloat(
      Vector=&RotateVectorByQuaternion(Vector To Rotate=<15 0 0>, Quaternion=$Particles.SampledRotation)
      Float=$Particles.NormalizedAge
     )
    )) {isSetParameters=true, module=SetVariables_#2}
   &Color(Scale Alpha=&FloatFromCurve(FloatCurve=DataInterface() {propertyValues {Curve {keys [<at 0 1>,<at 1 0>]}}}))
  ]
  rendererClasses [&NiagaraRibbonRendererProperties]
  renderers [{rendererClass=&NiagaraRibbonRendererProperties}]
 }
 Green_Decal {
  emitterUpdateScript [
   &EmitterState(Life Cycle Mode=::NewEnumerator1, Loop Behavior=::NewEnumerator1, Loop Duration=1, UseLoopDelay=false)
   &SpawnBurst_Instantaneous(Spawn Count=1, Use Spawn Probability=false)
  ]
  particleSpawnScript [&InitializeParticle(
    Lifetime=0.5
    Module.Write Color=true
    Module.Write Lifetime=true
    Module.Write Mass=true
    Module.Write Material Random=true
    Module.Write Position=false
    Module.Write Scale=false
    Module.Write SpriteRotation=false
    Module.Write SpriteSize=true
    Sprite Size=&Vector2DFromFloat(Value=30)
   )]
  particleUpdateScript [
   &ParticleState()
   &Color(
    Color=<rgba 1.4051 100 0 1>
    Scale Alpha=&FloatFromCurve(FloatCurve=DataInterface() {propertyValues {Curve {keys [{interpMode=RCIM_Cubic, tangentMode=RCTM_Break, time=0, value=1}, <at 1 0>]}}})
   )
   &CameraOffset(Camera Offset Amount=10)
   &SkeletalMeshLocation(Bone / Socket Index=9, Bone Sampling Mode=::NewEnumerator4, Inherit Velocity=false, MaskSpawnGroup=false, Skeletal Mesh=$User.SkeletalMesh)
  ]
  rendererClasses [&NiagaraSpriteRendererProperties]
  renderers [{rendererClass=&NiagaraSpriteRendererProperties}]
 }
 Green_Flame_Burst {
  emitterUpdateScript [
   &EmitterState(Life Cycle Mode=::NewEnumerator1, Loop Behavior=::NewEnumerator1, Loop Duration=0.25, UseLoopDelay=false)
   &SpawnBurst_Instantaneous(Spawn Count=25, Use Spawn Probability=false)
   &SpawnRate(SpawnRate=50, Use Spawn Probability=false)
  ]
  particleSpawnScript [
   &InitializeParticle(
    Lifetime=&RandomRangeFloat(Minimum=0.3, Maximum=0.45)
    Module.Write Color=true
    Module.Write Lifetime=true
    Module.Write Mass=true
    Module.Write Material Random=true
    Module.Write Position=false
    Module.Write Scale=false
    Module.Write SpriteRotation=false
    Module.Write SpriteSize=true
    Sprite Size=&Vector2DFromFloat(Value=&RandomRangeFloat(Minimum=15, Maximum=23))
   )
   &SubUVAnimation() {module=SubUVAnimation001}
  ]
  particleUpdateScript [
   &ParticleState()
   &SubUVAnimation()
   &SkeletalMeshLocation(
    Bone / Socket Index=9
    Bone Sampling Mode=::NewEnumerator4
    Inherit Velocity=false
    MaskSpawnGroup=false
    Sampled Velocity Coordinate Space:ENiagaraCoordinateSpace=::Local
    Skeletal Mesh=$User.SkeletalMesh
   )
   &AddVelocity(Velocity=&RandomRangeVector(Minimum=<-5 -36 -10>, Maximum=<30 36 10>))
   &Drag(Use Linear Drag=true, Use Rotational Drag=false)
   &SolveForcesAndVelocity(Clamp Velocity=false, Limit Acceleration=false)
   &Color(
    Color=&ColorFromCurve(ColorCurve=DataInterface() {propertyValues {
       AlphaCurve {keys [<at 0 1 RCIM_Cubic>,<at 1 1 RCIM_Cubic>]}
       BlueCurve {keys [<at 0 2.4014 RCIM_Cubic>,<at 1 0.2401 RCIM_Cubic>]}
       GreenCurve {keys [<at 0 50 RCIM_Cubic>,<at 1 5 RCIM_Cubic>]}
       RedCurve {keys [<at 0 7.797 RCIM_Cubic>,<at 1 0.7797 RCIM_Cubic>]}
      }})
    Scale Alpha=&FloatFromCurve(FloatCurve=DataInterface() {propertyValues {Curve {keys [
         {interpMode=RCIM_Cubic, tangentMode=RCTM_Break, time=0, value=0}
         {interpMode=RCIM_Cubic, tangentMode=RCTM_Break, time=0.1, value=1}
         {interpMode=RCIM_Cubic, tangentMode=RCTM_Break, time=1, value=0}
        ]}}})
   )
   &DynamicMaterialParameters(
    Index 0 Param 1=&FloatFromCurve(FloatCurve=DataInterface() {propertyValues {Curve {keys [<at 0 1>,<at 0.5 0>]}}})
    Index 0 Param 2=&RandomRangeFloat(Minimum=1, Maximum=3)
    Index 0 Param 3=&FloatFromCurve(FloatCurve=DataInterface() {propertyValues {Curve {keys [<at 0.5 1>,<at 1 4>]}}})
    Index 0 Param 4=&RandomRangeFloat(Minimum=0.8, Maximum=2)
    Module.Param0WriteEnabled=true
    Module.Param1WriteEnabled=true
    Module.Param2WriteEnabled=true
    Module.Param3WriteEnabled=true
   )
   &CameraOffset(Camera Offset Amount=10)
  ]
  rendererClasses [&NiagaraSpriteRendererProperties]
  renderers [{rendererClass=&NiagaraSpriteRendererProperties}]
 }
 Green_Flame_Burst001 {
  emitterUpdateScript [
   &EmitterState(Life Cycle Mode=::NewEnumerator1, Loop Behavior=::NewEnumerator1, Loop Duration=0.25, UseLoopDelay=false)
   &SpawnBurst_Instantaneous(Spawn Count=4, Use Spawn Probability=false)
   &SpawnRate(SpawnRate=15, Use Spawn Probability=false)
  ]
  particleSpawnScript [
   &InitializeParticle(
    Lifetime=&RandomRangeFloat(Minimum=0.2, Maximum=0.3)
    Module.Write Color=true
    Module.Write Lifetime=true
    Module.Write Mass=true
    Module.Write Material Random=true
    Module.Write Position=false
    Module.Write Scale=false
    Module.Write SpriteRotation=false
    Module.Write SpriteSize=true
    Sprite Size=&Vector2DFromFloat(Value=&RandomRangeFloat(Minimum=1, Maximum=80))
   )
   &SubUVAnimation() {module=SubUVAnimation001}
   &FN_SetSpawnGroup_v2(Number of Spawn Groups=9)
  ]
  particleUpdateScript [
   &ParticleState()
   &SubUVAnimation()
   &SkeletalMeshLocation(
    Bone / Socket Index=$Particles.SpawnGroup
    Bone Sampling Mode=::NewEnumerator4
    Inherit Velocity=false
    MaskSpawnGroup=false
    Sampled Velocity Coordinate Space:ENiagaraCoordinateSpace=::Local
    Skeletal Mesh=$User.SkeletalMesh
   )
   /SetVariables(Particles.Position=&Add_Vector(
     A=$Output.SkeletalMeshLocation.SampledPosition
     B=&Lerp_Vector(
      A=<0 0 0>
      B=&Multiply_VectorByFloat(Vector=&RandomRangeVector(Minimum=<0 -1 -1>, Maximum=<0 1 1>), Float=10)
      Alpha=&FloatFromCurve(FloatCurve=DataInterface() {propertyValues {Curve {keys [
           {arriveTangent=4.3131, interpMode=RCIM_Cubic, leaveTangent=4.3131, tangentMode=RCTM_User, time=0, value=0}
           {arriveTangent=0.3636, interpMode=RCIM_Cubic, leaveTangent=0.3636, tangentMode=RCTM_User, time=0.3, value=0.6656}
           <at 1 1 RCIM_Cubic>
          ]}}})
     )
    )) {isSetParameters=true}
   &AddVelocity(Velocity=&RandomRangeVector(Minimum=<-5 -30 -10>, Maximum=<30 30 10>)) {enabled=false}
   &Drag(Use Linear Drag=true, Use Rotational Drag=false)
   &SolveForcesAndVelocity(Clamp Velocity=false, Limit Acceleration=false)
   &Color(
    Color=&ColorFromCurve(ColorCurve=DataInterface() {propertyValues {
       AlphaCurve {keys [<at 0 1 RCIM_Cubic>,<at 1 1 RCIM_Cubic>]}
       BlueCurve {keys [<at 0 2.4014 RCIM_Cubic>,<at 1 0.2401 RCIM_Cubic>]}
       GreenCurve {keys [<at 0 50 RCIM_Cubic>,<at 1 5 RCIM_Cubic>]}
       RedCurve {keys [<at 0 7.797 RCIM_Cubic>,<at 1 0.7797 RCIM_Cubic>]}
      }})
    Scale Alpha=&FloatFromCurve(FloatCurve=DataInterface() {propertyValues {Curve {keys [
         {interpMode=RCIM_Cubic, tangentMode=RCTM_Break, time=0, value=0}
         {interpMode=RCIM_Cubic, tangentMode=RCTM_Break, time=0.1, value=1}
         {interpMode=RCIM_Cubic, tangentMode=RCTM_Break, time=1, value=0}
        ]}}})
   )
   &DynamicMaterialParameters(
    Index 0 Param 1=&FloatFromCurve(FloatCurve=DataInterface() {propertyValues {Curve {keys [<at 0 1>,<at 0.5 0>]}}})
    Index 0 Param 2=&RandomRangeFloat(Minimum=1, Maximum=3)
    Index 0 Param 3=&FloatFromCurve(FloatCurve=DataInterface() {propertyValues {Curve {keys [<at 0.5 1>,<at 1 4>]}}})
    Index 0 Param 4=&RandomRangeFloat(Minimum=0.8, Maximum=2)
    Module.Param0WriteEnabled=true
    Module.Param1WriteEnabled=true
    Module.Param2WriteEnabled=true
    Module.Param3WriteEnabled=true
   )
   &CameraOffset(Camera Offset Amount=10)
  ]
  rendererClasses [&NiagaraSpriteRendererProperties]
  renderers [{rendererClass=&NiagaraSpriteRendererProperties}]
 }
 Green_Flame_Continuous {
  bEnabled=false
  emitterUpdateScript [
   &EmitterState(Inactive Response=::NewEnumerator1, Life Cycle Mode=::NewEnumerator1, Loop Duration=1, UseLoopDelay=false)
   &SpawnRate(SpawnRate=3, Use Spawn Probability=false)
  ]
  particleSpawnScript [
   &InitializeParticle(
    Lifetime=&RandomRangeFloat(Minimum=0.7, Maximum=1)
    Module.Write Color=true
    Module.Write Lifetime=true
    Module.Write Mass=true
    Module.Write Material Random=true
    Module.Write Position=false
    Module.Write Scale=false
    Module.Write SpriteRotation=false
    Module.Write SpriteSize=true
    Sprite Size=&Vector2DFromFloat(Value=&RandomRangeFloat(Minimum=17, Maximum=23))
   )
   &SubUVAnimation() {module=SubUVAnimation001}
  ]
  particleUpdateScript [
   &ParticleState()
   &SubUVAnimation()
   &SkeletalMeshLocation(
    Bone / Socket Index=9
    Bone Sampling Mode=::NewEnumerator4
    Inherit Velocity=false
    MaskSpawnGroup=false
    Sampled Velocity Coordinate Space:ENiagaraCoordinateSpace=::Local
    Skeletal Mesh=$User.SkeletalMesh
   )
   &AddVelocity(Velocity=&RandomRangeVector(Minimum=<0 -25 0>, Maximum=<25 25 0>))
   &Drag(Drag=3, Use Linear Drag=true, Use Rotational Drag=false)
   &SolveForcesAndVelocity(Clamp Velocity=false, Limit Acceleration=false)
   &Color(
    Color=&ColorFromCurve(ColorCurve=DataInterface() {propertyValues {
       AlphaCurve {keys [<at 0 1 RCIM_Cubic>,<at 1 1 RCIM_Cubic>]}
       BlueCurve {keys [<at 0 0.4803 RCIM_Cubic>,<at 1 0.2401 RCIM_Cubic>]}
       GreenCurve {keys [<at 0 10 RCIM_Cubic>,<at 1 5 RCIM_Cubic>]}
       RedCurve {keys [<at 0 1.5594 RCIM_Cubic>,<at 1 0.7797 RCIM_Cubic>]}
      }})
    Scale Alpha=&FloatFromCurve(FloatCurve=DataInterface() {propertyValues {Curve {keys [
         {interpMode=RCIM_Cubic, tangentMode=RCTM_Break, time=0, value=0}
         {interpMode=RCIM_Cubic, tangentMode=RCTM_Break, time=0.1, value=1}
         {interpMode=RCIM_Cubic, tangentMode=RCTM_Break, time=1, value=0}
        ]}}})
   )
   &DynamicMaterialParameters(
    Index 0 Param 1=&FloatFromCurve(FloatCurve=DataInterface() {propertyValues {Curve {keys [<at 0 1>,<at 0.5 0>]}}})
    Index 0 Param 2=&RandomRangeFloat(Minimum=1, Maximum=3)
    Index 0 Param 3=&FloatFromCurve(FloatCurve=DataInterface() {propertyValues {Curve {keys [<at 0.5 1>,<at 1 4>]}}})
    Index 0 Param 4=&RandomRangeFloat(Minimum=0.8, Maximum=2)
    Module.Param0WriteEnabled=true
    Module.Param1WriteEnabled=true
    Module.Param2WriteEnabled=true
    Module.Param3WriteEnabled=true
   )
   &CameraOffset()
  ]
  rendererClasses [&NiagaraSpriteRendererProperties]
  renderers [{rendererClass=&NiagaraSpriteRendererProperties}]
 }
 Green_Flame_Continuous002 {
  bEnabled=false
  emitterUpdateScript [
   &EmitterState(Inactive Response=::NewEnumerator1, Life Cycle Mode=::NewEnumerator1, Loop Duration=1, UseLoopDelay=false)
   &SpawnRate(SpawnRate=7, Use Spawn Probability=false)
  ]
  particleSpawnScript [
   &InitializeParticle(
    Lifetime=&RandomRangeFloat(Minimum=0.7, Maximum=1)
    Module.Write Color=true
    Module.Write Lifetime=true
    Module.Write Mass=true
    Module.Write Material Random=true
    Module.Write Position=false
    Module.Write Scale=false
    Module.Write SpriteRotation=false
    Module.Write SpriteSize=true
    Sprite Size=&Vector2DFromFloat(Value=&RandomRangeFloat(Minimum=10, Maximum=14.5))
   )
   &SubUVAnimation() {module=SubUVAnimation001}
  ]
  particleUpdateScript [
   &ParticleState()
   &SubUVAnimation()
   &SkeletalMeshLocation(
    Bone / Socket Index=1
    Inherit Velocity=false
    MaskSpawnGroup=false
    Mesh Sampling Type=::NewEnumerator5
    Sampled Velocity Coordinate Space:ENiagaraCoordinateSpace=::Local
    Skeletal Mesh=DataInterface() {propertyValues {
      DefaultMesh=null
      ExcludeBoneName=root
      OptionalComponentClass=null
      PreviewMesh=&Broom_and_Pumpkin_Gadget
      SkinningMode=SkinOnTheFly
      SoftSourceActor=null
      SourceMode=Default
      UvSetIndex=0
      WholeMeshLOD=-1
      bExcludeBone=false
      bReadDeformedGeometry=true
      bRequireCurrentFrameData=true
      ComponentTags []
      FilteredBones [Pumpkin_02,bristle_02]
      FilteredSockets [RightEyeSocket, LeftEyeSocket]
      MeshUserParameter {parameter {name=None, typeDefHandle {registeredTypeIndex=1}, varData []}}
      SamplingRegions []
     }}
    Socket Sampling Mode=::NewEnumerator5
   )
   &AddVelocity(Velocity=&RandomRangeVector(Minimum=<-5 -5 0>, Maximum=<5 5 10>))
   &Drag(Drag=3, Use Linear Drag=true, Use Rotational Drag=false)
   &SolveForcesAndVelocity(Clamp Velocity=false, Limit Acceleration=false)
   &Color(
    Color=&ColorFromCurve(ColorCurve=DataInterface() {propertyValues {
       AlphaCurve {keys [<at 0 1 RCIM_Cubic>,<at 1 1 RCIM_Cubic>]}
       BlueCurve {keys [<at 0 2.4014 RCIM_Cubic>,<at 1 0.2401 RCIM_Cubic>]}
       GreenCurve {keys [<at 0 50 RCIM_Cubic>,<at 1 5 RCIM_Cubic>]}
       RedCurve {keys [<at 0 7.797 RCIM_Cubic>,<at 1 0.7797 RCIM_Cubic>]}
      }})
    Scale Alpha=&FloatFromCurve(FloatCurve=DataInterface() {propertyValues {Curve {keys [
         {interpMode=RCIM_Cubic, tangentMode=RCTM_Break, time=0, value=0}
         {interpMode=RCIM_Cubic, tangentMode=RCTM_Break, time=0.1, value=1}
         {interpMode=RCIM_Cubic, tangentMode=RCTM_Break, time=1, value=0}
        ]}}})
   )
   &DynamicMaterialParameters(
    Index 0 Param 1=&FloatFromCurve(FloatCurve=DataInterface() {propertyValues {Curve {keys [<at 0 1>,<at 0.5 0>]}}})
    Index 0 Param 2=&RandomRangeFloat(Minimum=1, Maximum=3)
    Index 0 Param 3=&FloatFromCurve(FloatCurve=DataInterface() {propertyValues {Curve {keys [<at 0.5 1>,<at 1 4>]}}})
    Index 0 Param 4=&RandomRangeFloat(Minimum=0.8, Maximum=2)
    Module.Param0WriteEnabled=true
    Module.Param1WriteEnabled=true
    Module.Param2WriteEnabled=true
    Module.Param3WriteEnabled=true
   )
   &CameraOffset()
  ]
  rendererClasses [&NiagaraSpriteRendererProperties]
  renderers [{rendererClass=&NiagaraSpriteRendererProperties}]
 }
 Green_Flame_Eyes {
  emitterUpdateScript [
   &EmitterState(
    Inactive Response=::NewEnumerator1
    Life Cycle Mode=::NewEnumerator1
    Loop Behavior=::NewEnumerator1
    Loop Duration Mode=::NewEnumerator1
    UseLoopDelay=false
   )
   &SpawnBurst_Instantaneous(Spawn Count=1, Use Spawn Probability=false)
  ]
  particleSpawnScript [
   &InitializeParticle(
    Color=<rgba 1 1 1 0.5>
    Lifetime=&RandomRangeFloat(Minimum=0.7, Maximum=1)
    Mesh Scale=&Multiply_VectorByFloat(Vector=<1 1 1>, Float=0.8)
    Module.Write Color=true
    Module.Write Lifetime=true
    Module.Write Mass=true
    Module.Write Material Random=true
    Module.Write Position=false
    Module.Write Scale=true
    Module.Write SpriteRotation=false
    Module.Write SpriteSize=true
    Sprite Size=&Vector2DFromFloat(Value=&RandomRangeFloat(Minimum=10, Maximum=14.5))
   )
   /Niagara/Modules/Spawn/Orientation/V2/InitialMeshOrientation(
    Orientation Axis:Vector3f=<0 1 0>
    Orientation Coordinate Space:ENiagaraCoordinateSpace=::Local
    Orientation Vector=<-1 0 0>
    Rotation=<0 0.5 0>
    Use Orientation Vector=true
    Use Rotation Vector=true
   )
  ]
  particleUpdateScript [
   &ParticleState(Kill Particles When Lifetime Has Elapsed=false)
   &SkeletalMeshLocation(
    Inherit Velocity=false
    MaskSpawnGroup=false
    Mesh Sampling Type=::NewEnumerator5
    Sampled Position Offset=<3 8.5 -5>
    Sampled Position Offset Coordinate Space:ENiagaraMeshTransforms=::NewEnumerator3
    Skeletal Mesh=$User.SkeletalMesh
    Socket Sampling Mode=::NewEnumerator5
   )
   &Color(Color=&ColorFromCurve(ColorCurve=DataInterface() {propertyValues {
       AlphaCurve {keys [<at 0 1 RCIM_Cubic>,<at 1 1 RCIM_Cubic>]}
       BlueCurve {keys [<at 0 2.4014 RCIM_Cubic>,<at 1 0.2401 RCIM_Cubic>]}
       GreenCurve {keys [<at 0 50 RCIM_Cubic>,<at 1 5 RCIM_Cubic>]}
       RedCurve {keys [<at 0 7.797 RCIM_Cubic>,<at 1 0.7797 RCIM_Cubic>]}
      }}))
   &DynamicMaterialParameters(
    Index 0 Param 1=&FloatFromCurve(FloatCurve=DataInterface() {propertyValues {Curve {keys [{arriveTangent=5.9932, interpMode=RCIM_Cubic, leaveTangent=5.9932, tangentMode=RCTM_User, time=0, value=0}, <at 0.5 1 RCIM_Cubic>]}}})
    Index 0 Param 2=&FloatFromCurve(FloatCurve=DataInterface() {propertyValues {Curve {keys [<at 0 0>,<at 1 1>]}}})
    Index 0 Param 3=&FloatFromCurve(FloatCurve=DataInterface() {propertyValues {Curve {keys [<at 0.5 1>,<at 1 4>]}}})
    Index 0 Param 4=&RandomRangeFloat(Minimum=1, Maximum=3)
    Module.Param0WriteEnabled=true
    Module.Param1WriteEnabled=true
    Module.Param2WriteEnabled=true
    Module.Param3WriteEnabled=true
   )
   &CameraOffset(Camera Offset Amount=4)
  ]
  rendererClasses [&NiagaraMeshRendererProperties]
  renderers [{rendererClass=&NiagaraMeshRendererProperties, rendererIndex=0}]
 }
 Green_Radial_Light {
  emitterUpdateScript [
   &EmitterState(Life Cycle Mode=::NewEnumerator1, Loop Behavior=::NewEnumerator1, Loop Duration=1, UseLoopDelay=false)
   &SpawnBurst_Instantaneous(Spawn Count=1, Use Spawn Probability=false)
  ]
  particleSpawnScript [&InitializeParticle(
    Lifetime=0.3
    Module.Write Color=true
    Module.Write Lifetime=true
    Module.Write Mass=true
    Module.Write Material Random=true
    Module.Write Position=false
    Module.Write Scale=false
    Module.Write SpriteRotation=false
    Module.Write SpriteSize=true
    Sprite Size=&Vector2DFromFloat(Value=50)
   )]
  particleUpdateScript [
   &ParticleState()
   &SkeletalMeshLocation(Bone / Socket Index=9, Bone Sampling Mode=::NewEnumerator4, Inherit Velocity=false, MaskSpawnGroup=false, Skeletal Mesh=$User.SkeletalMesh)
   &Color(
    Color=<rgba 0.5556 1 0.0516 0.6>
    Scale Alpha=&FloatFromCurve(FloatCurve=DataInterface() {propertyValues {Curve {keys [<at 0.5 1>,<at 1 0>]}}})
   )
   &CameraOffset(Camera Offset Amount=30)
  ]
  rendererClasses [&NiagaraSpriteRendererProperties, &NiagaraLightRendererProperties]
  renderers [
   {rendererClass=&NiagaraSpriteRendererProperties}
   {rendererClass=&NiagaraLightRendererProperties, rendererIndex=1}
  ]
 }
 Green_Wave {
  emitterUpdateScript [
   &EmitterState(Life Cycle Mode=::NewEnumerator1, Loop Behavior=::NewEnumerator1, Loop Duration=1, UseLoopDelay=false)
   &SpawnBurst_Instantaneous(Spawn Count=1, Use Spawn Probability=false)
  ]
  particleSpawnScript [&InitializeParticle(
    Lifetime=0.3
    Mesh Scale=&VectorFromFloat(Value=0.12)
    Module.Write Color=true
    Module.Write Lifetime=true
    Module.Write Mass=true
    Module.Write Material Random=true
    Module.Write Position=false
    Module.Write Scale=true
    Module.Write SpriteRotation=false
    Module.Write SpriteSize=true
    Sprite Size=<5 5>
   )]
  particleUpdateScript [
   &ParticleState()
   &Color(
    Color=<rgba 3.5 3.5 3.5 1>
    Scale Alpha=&FloatFromCurve(FloatCurve=DataInterface() {propertyValues {Curve {keys [{interpMode=RCIM_Cubic, tangentMode=RCTM_Break, time=0, value=1}, <at 1 0>]}}})
   )
   &SkeletalMeshLocation(Bone / Socket Index=9, Bone Sampling Mode=::NewEnumerator4, Inherit Velocity=false, MaskSpawnGroup=false, Skeletal Mesh=$User.SkeletalMesh)
   &ScaleMeshSize(Scale Factor=&VectorFromFloat(Value=&FloatFromCurve(FloatCurve=DataInterface() {propertyValues {Curve {keys [{interpMode=RCIM_Cubic, tangentMode=RCTM_Break, time=0, value=0}, <at 1 1>]}}})))
   &CameraOffset(Camera Offset Amount=50)
  ]
  rendererClasses [&NiagaraMeshRendererProperties]
  renderers [{rendererClass=&NiagaraMeshRendererProperties, rendererIndex=0}]
 }
 Smoke_Reveal {
  emitterUpdateScript [
   &EmitterState(Inactive Response=::NewEnumerator1, Life Cycle Mode=::NewEnumerator1, Loop Behavior=::NewEnumerator1, Loop Duration=1, UseLoopDelay=false)
   &SpawnBurst_Instantaneous(Spawn Count=18, Use Spawn Probability=false)
  ]
  particleSpawnScript [
   &InitializeParticle(
    Lifetime=&RandomRangeFloat(Minimum=0.35, Maximum=0.8)
    Module.Write Color=true
    Module.Write Lifetime=true
    Module.Write Mass=true
    Module.Write Material Random=true
    Module.Write Position=false
    Module.Write Scale=false
    Module.Write SpriteRotation=true
    Module.Write SpriteSize=true
    Sprite Rotation=&RandomRangeFloat(Minimum=0, Maximum=360)
    Sprite Size=&Vector2DFromFloat(Value=&RandomRangeFloat(Minimum=40, Maximum=100))
   )
   &FN_SetSpawnGroup_v2(Number of Spawn Groups=9)
  ]
  particleUpdateScript [
   &ParticleState()
   &SkeletalMeshLocation(
    Apply Sampled Orientation to Mesh Particles=true
    Bone / Socket Index=$Particles.SpawnGroup
    Bone Sampling Mode=::NewEnumerator4
    Inherit Velocity=false
    MaskSpawnGroup=false
    Sampled Position Offset=&RandomRangeVector(Minimum=<-1 -1 -1>, Maximum=<1 1 1>)
    Sampled Position Offset Coordinate Space:ENiagaraMeshTransforms=::NewEnumerator3
    Skeletal Mesh=$User.SkeletalMesh
   )
   /SetVariables(Particles.Position=&Add_Vector(
     A=$Output.SkeletalMeshLocation.SampledPosition
     B=&Lerp_Vector(
      A=<0 0 0>
      B=&Multiply_VectorByFloat(Vector=&RandomRangeVector(Minimum=<0 -1 -1>, Maximum=<0 1 1>), Float=50)
      Alpha=&FloatFromCurve(FloatCurve=DataInterface() {propertyValues {Curve {keys [
           {arriveTangent=4.3131, interpMode=RCIM_Cubic, leaveTangent=4.3131, tangentMode=RCTM_User, time=0, value=0}
           {arriveTangent=0.3636, interpMode=RCIM_Cubic, leaveTangent=0.3636, tangentMode=RCTM_User, time=0.3, value=0.6656}
           <at 1 1 RCIM_Cubic>
          ]}}})
     )
    )) {isSetParameters=true}
   &Color(
    Color=&ColorFromCurve(ColorCurve=DataInterface() {propertyValues {
       AlphaCurve {keys [<at 0 1 RCIM_Cubic>,<at 1 1 RCIM_Cubic>]}
       BlueCurve {keys [<at 0 0.0516 RCIM_Cubic>,<at 0.9999 0.6719 RCIM_Cubic>]}
       GreenCurve {keys [<at 0 1 RCIM_Cubic>,<at 0.9999 0.6719 RCIM_Cubic>]}
       RedCurve {keys [<at 0 0.5556 RCIM_Cubic>,<at 0.9999 0.6719 RCIM_Cubic>]}
      }})
    Scale Alpha=&FloatFromCurve(FloatCurve=DataInterface() {propertyValues {Curve {keys [
         {interpMode=RCIM_Cubic, tangentMode=RCTM_Break, time=0, value=0}
         {arriveTangent=2.796, interpMode=RCIM_Cubic, leaveTangent=2.796, tangentMode=RCTM_Break, time=0.148, value=0.736}
         {arriveTangent=0.3761, interpMode=RCIM_Cubic, leaveTangent=0.3761, tangentMode=RCTM_Break, time=0.353, value=0.987}
         {interpMode=RCIM_Cubic, tangentMode=RCTM_Break, time=1, value=0}
        ]}}})
   )
   &SubUVAnimation()
   &DynamicMaterialParameters(
    Index 0 Param 1=&FloatFromCurve(FloatCurve=DataInterface() {propertyValues {Curve {keys [<at 0.5 1>,<at 1 -1>]}}})
    Index 0 Param 2=1
    Index 0 Param 3=1
    Index 0 Param 4=0.5
    Module.Param0WriteEnabled=true
    Module.Param1WriteEnabled=true
    Module.Param2WriteEnabled=true
    Module.Param3WriteEnabled=true
   )
   &SpriteRotationRate(Rotation Rate=0.2)
   &ScaleSpriteSize(Scale Factor=&Vector2DFromFloat(Value=&FloatFromCurve(FloatCurve=DataInterface() {propertyValues {Curve {keys [<at 0 0.3>,<at 1 1>]}}})))
   &CameraOffset(Camera Offset Amount=25)
  ]
  rendererClasses [&NiagaraSpriteRendererProperties]
  renderers [{rendererClass=&NiagaraSpriteRendererProperties}]
 }
 Sparkles {
  bEnabled=false
  emitterUpdateScript [
   &EmitterState(
    Delay First Loop Only=true
    Inactive Response=::NewEnumerator1
    Life Cycle Mode=::NewEnumerator1
    Loop Delay=0.1
    Loop Duration=1
    UseLoopDelay=true
   )
   &SpawnRate(SpawnRate=25, Use Spawn Probability=false)
  ]
  particleSpawnScript [&InitializeParticle(
    Lifetime=&RandomRangeFloat(Minimum=0.5, Maximum=1)
    Module.Write Color=true
    Module.Write Lifetime=true
    Module.Write Mass=true
    Module.Write Material Random=true
    Module.Write Position=false
    Module.Write Scale=false
    Module.Write SpriteRotation=false
    Module.Write SpriteSize=true
    Sprite Size=<5 5>
   )]
  particleUpdateScript [
   &ParticleState()
   &SkeletalMeshLocation(
    Bone Sampling Mode=::NewEnumerator4
    Inherit Velocity=false
    MaskSpawnGroup=false
    Sampled Position Coordinate Space=::Local
    Skeletal Mesh=$User.SkeletalMesh
   )
   &SphereLocation(
    MaskSpawnGroup=false
    Override Seed=false
    Sphere Origin:NiagaraPosition=$Particles.Position
    Sphere Radius=10
    Surface Only=false
    UseNonUniformScale=false
    UseOffset=false
    UseSphereOrientation=false
   )
   /SetVariables(Particles.Position=&Add_Vector(
     A=Linked() {linkedVariable {name=Particles.Position, type=NiagaraPosition}}
     B=&Multiply_VectorByFloat(
      Vector=&RotateVectorByQuaternion(Vector To Rotate=<0 -125 0>, Quaternion=$Particles.SampledRotation)
      Float=$Particles.NormalizedAge
     )
    )) {isSetParameters=true}
   &SolveForcesAndVelocity(Clamp Velocity=false, Limit Acceleration=false)
   &Color(
    Color=&ColorFromCurve(ColorCurve=DataInterface() {propertyValues {
       AlphaCurve {keys [<at 0 1 RCIM_Cubic>,<at 1 1 RCIM_Cubic>]}
       BlueCurve {keys [<at 0 0.048 RCIM_Cubic>,<at 1 0.6928 RCIM_Cubic>]}
       GreenCurve {keys [<at 0 1 RCIM_Cubic>,<at 1 1 RCIM_Cubic>]}
       RedCurve {keys [<at 0 0.1559 RCIM_Cubic>,<at 1 0 RCIM_Cubic>]}
      }})
    Scale Alpha=&FloatFromCurve(FloatCurve=DataInterface() {propertyValues {Curve {keys [<at 0 0>,<at 0.2901 1>,<at 1 1>]}}})
   )
   &ScaleSpriteSize(Scale Factor=&Vector2DFromFloat(Value=&FloatFromCurve(FloatCurve=DataInterface() {propertyValues {Curve {keys [
          {interpMode=RCIM_Cubic, tangentMode=RCTM_Break, time=0, value=0}
          {interpMode=RCIM_Cubic, tangentMode=RCTM_Break, time=0.3, value=3.5}
          {interpMode=RCIM_Cubic, tangentMode=RCTM_Break, time=1, value=0}
         ]}}})))
  ]
  rendererClasses [&NiagaraSpriteRendererProperties]
  renderers [{rendererClass=&NiagaraSpriteRendererProperties}]
 }
}
summary {
 systemName=NS_HildaPumpkin
 emitters [
  {
   emitterName=Green_Radial_Light
   rendererClasses [&NiagaraSpriteRendererProperties, &NiagaraLightRendererProperties]
  }
  {emitterName=Green_Wave, rendererClasses [&NiagaraMeshRendererProperties]}
  {emitterName=Green_Flame_Burst, rendererClasses [&NiagaraSpriteRendererProperties]}
  {emitterName=Green_Decal, rendererClasses [&NiagaraSpriteRendererProperties]}
  {bEnabled=false, emitterName=Embers, rendererClasses [&NiagaraSpriteRendererProperties]}
  {bEnabled=false, emitterName=Green_Flame_Continuous, rendererClasses [&NiagaraSpriteRendererProperties]}
  {emitterName=Green_Flame_Eyes, rendererClasses [&NiagaraMeshRendererProperties]}
  {bEnabled=false, emitterName=Green_Flame_Continuous002, rendererClasses [&NiagaraSpriteRendererProperties]}
  {emitterName=Green_Broom_Trail, rendererClasses [&NiagaraRibbonRendererProperties]}
  {bEnabled=false, emitterName=Sparkles, rendererClasses [&NiagaraSpriteRendererProperties]}
  {emitterName=Smoke_Reveal, rendererClasses [&NiagaraSpriteRendererProperties]}
  {emitterName=Green_Flame_Burst001, rendererClasses [&NiagaraSpriteRendererProperties]}
  {bEnabled=false, emitterName=Green_Broom_Trail001, rendererClasses [&NiagaraRibbonRendererProperties]}
 ]
 userVariables [
  User.Ribbon Length {type=NiagaraFloat, defaultValue=0, description=""}
  User.Ribbon Width {type=NiagaraFloat, defaultValue=0, description=""}
  User.RibbonIDOffset {type=NiagaraInt32, defaultValue=1}
  User.SkeletalMesh {
   type=NiagaraDataInterfaceSkeletalMesh
   defaultValue {
    struct=NiagaraExt_VariableValue_DataInterface
    value {dataInterface=&self:NiagaraDataInterfaceSkeletalMesh_6}
   }
  }
 ]
}
